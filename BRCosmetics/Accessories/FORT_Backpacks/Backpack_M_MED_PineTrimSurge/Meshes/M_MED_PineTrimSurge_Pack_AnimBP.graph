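# Animation Blueprint for the PineTrimSurge backpack cosmetic, adjusting the pack to the wearer's spine bend and skydiving state.
# Each update maps the yaw of the owner mesh's spine_05-to-spine_01 delta from reference pose into a blend alpha, and eases a skydiving alpha via FInterpTo.
# Three chained additive Modify Bone nodes apply a fixed pack flip, a spine-driven pack tilt and shift, and a skydive root shift with slight shrink.
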AnimBlueprint /BRCosmetics/Accessories/FORT_Backpacks/Backpack_M_MED_PineTrimSurge/Meshes/M_MED_PineTrimSurge_Pack_AnimBP
target_skeleton=&/M_MED_PineTrimSurge_Pack_Skeleton
graphs [
 AnimGraph {nodes [
   n1 = "Transform (Modify) Bone - Bone: pack"(ComponentPose=Local To Component(LocalPose=Input Pose()), Translation="0,-1,0", Rotation="0,180,0") {
    rotationMode=BMM_Additive
    rotationSpace=BCS_BoneSpace
    scaleMode=BMM_Additive
    scaleSpace=BCS_BoneSpace
    translationMode=BMM_Additive
    translationSpace=BCS_BoneSpace
    alphaBoolBlend {blendOption=Linear}
    boneToModify {boneName=pack}
    rotation {yaw=180}
    translation {y=-1}
   }
   n2 = "Transform (Modify) Bone - Bone: pack"(ComponentPose=@n1, Alpha=Get BackpackRotate() -> (BackpackRotate:double), Translation="0,13,5", Rotation="0,0,-35") {
    rotationMode=BMM_Additive
    rotationSpace=BCS_BoneSpace
    scaleMode=BMM_Additive
    scaleSpace=BCS_BoneSpace
    translationMode=BMM_Additive
    translationSpace=BCS_BoneSpace
    alphaBoolBlend {blendOption=Linear}
    boneToModify {boneName=pack}
    rotation {roll=-35}
    translation {y=13, z=5}
   }
   n3 = "Transform (Modify) Bone - Bone: root"(
    ComponentPose=@n2
    Alpha=Get MoveWhileSkydiveDown() -> (MoveWhileSkydiveDown:double)
    Translation="0,-3,5"
    Rotation="0,0,0"
    Scale="0.95,0.95,0.95"
   ) {
    rotationMode=BMM_Additive
    scaleMode=BMM_Additive
    translationMode=BMM_Additive
    alphaBoolBlend {blendOption=Linear}
    boneToModify {boneName=root}
    scale=<0.95 0.95 0.95>
    translation {y=-3, z=5}
   }
   Output Pose(Result=Component To Local(ComponentPose=@n3)) {layerGroup=None}
  ]}
 EventGraph {nodes [
   n1 = Event BlueprintUpdateAnimation()
   n2 = BreakTransform(InTransform=GetDeltaTransformFromRefPose(self=GetOwnerSkeletalMeshComponent(), BoneName=spine_05, BaseName=spine_01))
   > Set BackpackRotate(BackpackRotate:double=MapRangeClamped(Value=BreakRotator(InRot=@n2.Rotation).Yaw, InRangeA=-25, InRangeB=-65)) -> (Output_Get:double)
   n3 = Get bIsSkyDiving()
   > Set MoveWhileSkydiveDown(MoveWhileSkydiveDown:double=FInterpTo(Current=Get MoveWhileSkydiveDown(), Target="To Float (Boolean)"(InBool=@n3), DeltaTime=@n1.DeltaTimeX, InterpSpeed=15)) -> (Output_Get:double)
   TryGetPawnOwner()
  ]}
]
variables [
 BackpackRotate {type=number}
 MoveWhileSkydiveDown {type=number}
 NewVar {type=Transform, default {location=<0 0 0>, rotation=<rot 0 0 0>, scale=<1 1 1>}}
 NewVar_0 {type=Transform, default {location=<0 0 0>, rotation=<rot 0 0 0>, scale=<1 1 1>}}
]
